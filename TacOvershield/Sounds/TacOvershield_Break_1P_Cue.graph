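# TacOvershield_Break_1P_Cue is a SoundCue that plays the first-person overshield break sound.
# A single Wave Player feeds TacOvershield_Break_1P_02 straight to the output at reduced volume, with no randomization or modulation.
# It uses the 2D hit-notification shield-break receive sound class and overrides concurrency to one instance, stopping the oldest when retriggered.
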SoundCue /TacOvershield/Sounds/TacOvershield_Break_1P_Cue
root=WavePlayer_0
nodes [{id=WavePlayer_0, properties {soundWaveAssetPtr=&/TacOvershield_Break_1P_02}}]
properties {
 bOverrideConcurrency=true
 maxDistance=2097152
 soundClassObject=/Game/Sounds/FortSoundClassesAndModes/Fort_2DHitNotifications_Receive_Shield_Break_PSM
 volumeMultiplier=0.55
 concurrencyOverrides {maxCount=1, resolutionRule=StopOldest}
}
