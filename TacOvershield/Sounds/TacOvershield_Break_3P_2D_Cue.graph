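# Sound cue for the TacOvershield break, a 2D third-person sound routed to the shield-break hit notification sound class.
# A single wave player feeds a modulator that slightly randomizes pitch while keeping volume constant, under a reduced cue volume.
# It references the shared critical hit-notification concurrency set and also carries a single-voice, stop-oldest override.
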SoundCue /TacOvershield/Sounds/TacOvershield_Break_3P_2D_Cue
root=Modulator_1
nodes [
 {id=Modulator_1, children [WavePlayer_0], properties {pitchMax=1.1, pitchMin=0.9, volumeMax=1, volumeMin=1}}
 {id=WavePlayer_0, properties {soundWaveAssetPtr=&/TacOvershield_Break_3P_2D_01}}
]
properties {
 maxDistance=2097152
 soundClassObject=/Game/Sounds/FortSoundClassesAndModes/Fort_2DHitNotifications_Receive_Shield_Break_PSM
 volumeMultiplier=0.65
 concurrencyOverrides {maxCount=1, resolutionRule=StopOldest}
 concurrencySet [/Game/Sounds/Concurrency_Rules/HitNotifications_Critical_StopOldest]
}
