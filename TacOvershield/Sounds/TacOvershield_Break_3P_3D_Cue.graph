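# Sound cue for the third-person 3D break sound of the TacOvershield shield.
# A single wave plays through a modulator that slightly randomizes pitch at constant volume.
# It uses the Fort shield-break 3D attenuation and the Fort_Impacts sound class, with boosted volume.
# A concurrency override caps simultaneous instances, stopping the lowest priority, and takes precedence over the assigned hit-notification concurrency set.
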SoundCue /TacOvershield/Sounds/TacOvershield_Break_3P_3D_Cue
root=Modulator_0
nodes [
 {id=Modulator_0, children [WavePlayer_1], properties {pitchMax=1.1, pitchMin=0.9, volumeMax=1, volumeMin=1}}
 {id=WavePlayer_1, properties {soundWaveAssetPtr=&/TacOvershield_Break_3P_3D_01}}
]
properties {
 attenuationSettings=/Game/Sounds/Attenuation/Fort_Shield_Break_3D_Att
 bOverrideConcurrency=true
 maxDistance=8500
 soundClassObject=&Fort_Impacts
 volumeMultiplier=1.5
 concurrencyOverrides {maxCount=2, resolutionRule=StopLowestPriority}
 concurrencySet [&HitNotifications_TwoMaximum_StopOldest_VolumeScale]
}
