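# TacOvershield_Loop_Cue is a Sound Cue named for the looping audio of the TacOvershield item, routed through the Fort_Gadgets_Items sound class.
# It has no root node and no nodes, so it plays nothing as it stands.
# It only carries a reduced volume multiplier and a very large max distance.
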SoundCue /TacOvershield/Sounds/TacOvershield_Loop_Cue
root=null
nodes []
properties {maxDistance=2097152, soundClassObject=&Fort_Gadgets_Items, volumeMultiplier=0.3}
