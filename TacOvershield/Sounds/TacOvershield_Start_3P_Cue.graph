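# Sound cue for the third-person Tac Overshield start sound, routed to the gadgets and items sound class.
# A Random node picks one of three equally weighted start wave variants and feeds a Modulator at default settings for slight pitch and volume variation.
# A dedicated 3P attenuation asset spatializes it at reduced volume, and a concurrency override allows only one instance at a time.
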SoundCue /TacOvershield/Sounds/TacOvershield_Start_3P_Cue
root=Modulator_0
nodes [
 {id=Modulator_0, children [Random_0]}
 {id=Random_0, children [WavePlayer_1,WavePlayer_2,WavePlayer_3], properties {bShouldExcludeFromBranchCulling=true, weights [1,1,1]}}
 {id=WavePlayer_1, properties {soundWaveAssetPtr=&/TacOvershield_Start_3P_01}}
 {id=WavePlayer_2, properties {soundWaveAssetPtr=&/TacOvershield_Start_3P_02}}
 {id=WavePlayer_3, properties {soundWaveAssetPtr=&/TacOvershield_Start_3P_03}}
]
properties {
 attenuationSettings=&/TacOvershield_3P_Att
 bOverrideConcurrency=true
 maxDistance=2600
 soundClassObject=&Fort_Gadgets_Items
 volumeMultiplier=0.7
 concurrencyOverrides {maxCount=1}
}
