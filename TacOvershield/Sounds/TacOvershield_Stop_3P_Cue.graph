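# Third-person stop sound cue for the TacOvershield, in the Fortnite gadgets and items sound class.
# The cue has no nodes and a null root, so it produces no audio as it stands.
# It still carries playback settings: bullet impact attenuation, reduced volume, a max distance and a concurrency override limiting it to a single instance.
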SoundCue /TacOvershield/Sounds/TacOvershield_Stop_3P_Cue
root=null
nodes []
properties {
 attenuationSettings=/Game/Sounds/Attenuation/Fort_Bullet_Impact_Att
 bOverrideConcurrency=true
 maxDistance=5500
 soundClassObject=&Fort_Gadgets_Items
 volumeMultiplier=0.4
 concurrencyOverrides {maxCount=1}
}
