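# TacOvershield stop sound cue, which by its name plays when the tactical overshield stops.
# A single wave player feeds the root directly with the TacOvershield_Stop wave, with no randomization or mixing.
# The cue plays at reduced volume under the Fort gadgets items sound class and sets no attenuation asset.
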SoundCue /TacOvershield/Sounds/TacOvershield_Stop_Cue
root=WavePlayer_0
nodes [{id=WavePlayer_0, properties {soundWaveAssetPtr=&/TacOvershield_Stop}}]
properties {maxDistance=2097152, soundClassObject=&Fort_Gadgets_Items, volumeMultiplier=0.4}
